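; Parent player-pawn Blueprint for Tandem NPCs in base activities, handling downed (DBNO) protection, revival and a friendly marker.
; Inside the safe zone, the server watches DBNO changes, blocks carrying via a RepNotify and applies a pre-rebirth GameplayEffect during activities.
; Async waits on defense or reclaim completion events trigger revival, swapping in a post-rebirth effect.
; Alert-level changes toggle the overhead marker on clients, distinguishing fighting NPCs from players.
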(event EventBeginPlay
  (bind _self self)
  (bind _returnvalue (Ability|GetFortAbilitySystemComponent _self))
  ; Wait for the NPC to go DBNO
  (|CallParentFunction)
  (if (Variables|Gameplay|GetbIsInsideSafeZone)
    (Networking|SwitchHasAuthority
      (:Authority
        (Utilities|IsValid _returnvalue
          (:Is%20Valid
            (Variables|Default|SetTandemAbilitySystem _returnvalue)
            (EventDispatchers|BindEventtoOnDBNOStateChanged (delegate Custom|DBNOStateChanged)))))
      (:Remote
        (CallFunction|ToggleIconforAlertLevel (Variables|Default|GetAlertLevel))))))

(event Custom|DBNOStateChanged
  (params (FortPawn :type Fort%20Pawn%20Object%20Reference) (bInIsDBNO :type Boolean))
  ; Apply damage immunity effects when the NPC is DBNO and wait for the base activity event(s) to complete if they are in an activity
  (Networking|SwitchHasAuthority
    (:Authority
      (if bInIsDBNO
        (if (Variables|Gameplay|GetbIsInsideSafeZone)
          (Variables|Default|SetDBNOBlockCarry true)
          (if (GameplayTags|HasMatchingGameplayTag self "(TagName=\"TidalCrane.NPC.InBaseActivity\")")
            (bind _returnvalue (GameplayEffects|ApplyGameplayEffectToSelf :self (Variables|Default|GetTandemAbilitySystem) :GameplayEffectClass (Variables|Default|GetPreRebirthGE) :Level 0.0))
            (Variables|Default|SetPreRebirthGEHandle _returnvalue)
            (Ability|Async|WaitGameplayEventToActor 0 (Variables|Default|GetDefenseCompleteTag) true
              (:then
                (Ability|Async|WaitGameplayEventToActor 0 (Variables|Default|GetReclaimCompleteTag) true
                  (:EventReceived
                    _)))
              (:EventReceived
                _))
            (CallFunction|ReviveNPC)
            (else
              (FortPawn|SetHealth 1.0)))
          (else
            (FortPawn|SetHealth 1.0)))
        (else
          (GameplayEffects|RemoveActiveGameplayEffect (Variables|Default|GetTandemAbilitySystem) (Variables|Default|GetPreRebirthGEHandle) -1)
          (FortPawn|SetHealth (FortPawn|GetMaxHealth))
          (FortPawn|SetShield (FortPawn|GetMaxShield)))))))

(event Custom|ReviveNPC
  ; When a base activity event is complete, revive the NPC if they are DBNO
  (Networking|SwitchHasAuthority
    (:Authority
      (GameplayEffects|RemoveActiveGameplayEffect (Variables|Default|GetTandemAbilitySystem) (Variables|Default|GetPreRebirthGEHandle) -1)
      (if (Variables|Gameplay|GetbIsInsideSafeZone)
        (FortPlayerPawn|ManualRevivefromDBNO 0)
        (FortPawn|SetHealth (FortPawn|GetMaxHealth))
        (FortPawn|SetShield (FortPawn|GetMaxShield))
        (else
          (FortPawn|SetHealth 1.0)))
      (GameplayEffects|ApplyGameplayEffectToSelf :self (Variables|Default|GetTandemAbilitySystem) :GameplayEffectClass (Variables|Default|GetPostRebirthGE) :Level 0.0)
      (EventDispatchers|UnbindEventfromOnDBNOStateChanged self (EventDispatchers|CreateEvent :!event Custom|DBNOStateChanged)))))

(event Custom|DisableCarry
  ; Disable DBNO carry
  (bind _asfort_ability_system_component (Utilities|Casting|CastToFortAbilitySystemComponent (Ability|GetAbilitySystemComponent self)))
  (Tags|BlockAbilitiesWithTags _asfort_ability_system_component (Variables|Default|GetAbilityTagToBlock)))

(event Collision|EventActorBeginOverlap
  (params (OtherActor :type Actor%20Object%20Reference))
  (|CallParentFunction OtherActor) ; This node is disabled and will not be called.\nDrag off pins to build functionality.
)

(event EventOnDamageServer
  (params (Damage :type Float%20%28single-precision%29) (DamageTags :type Gameplay%20Tag%20Container%20Structure%20%28by%20ref%29) (Momentum :type Vector :default "0, 0, 0") (HitInfo :type Hit%20Result%20Structure%20%28by%20ref%29) (InstigatedBy :type Controller%20Object%20Reference) (DamageCauser :type Actor%20Object%20Reference) (EffectContext :type Gameplay%20Effect%20Context%20Handle%20Structure))
  (|CallParentFunction Damage DamageTags Momentum HitInfo InstigatedBy DamageCauser EffectContext))

(fn ClientOnAlertLevelChanged
  (params (OldAlertLevel :type EAlertLevel%20Enum :default "Unaware") (NewAlertLevel :type EAlertLevel%20Enum :default "Unaware"))
  (|CallParentFunction OldAlertLevel NewAlertLevel)
  (CallFunction|ToggleIconforAlertLevel NewAlertLevel))

(fn OnRep_DBNOBlockCarry
  (if (Variables|Default|GetDBNOBlockCarry)
    (CallFunction|DisableCarry)))

(fn ToggleIconForAlertLevel
  (params (AlertLevel :type EAlertLevel%20Enum :default "Unaware"))
  ; When a base activity NPC enters combat, put the friendly marker above their head so players don't mistake them for other players. Remove marker when they aren't in combat so it doesn't interfere with the speech icon
  (switch Utilities|FlowControl|Switch|SwitchonEAlertLevel AlertLevel
    (:Unaware
      (Rendering|SetVisibility (Variables|Default|GetFriendlyMarker) false false))
    (:Suspicious
      _)
    (:LostTarget
      _)
    (:Threatened
      _))
  (Rendering|SetVisibility (Variables|Default|GetFriendlyMarker) true false))

(fn ConstructionScript
  (|CallParentFunction))
